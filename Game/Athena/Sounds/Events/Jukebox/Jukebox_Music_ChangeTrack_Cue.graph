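# Sound cue that plays the jukebox change-track sound in Athena, named as a music event.
# A single wave player feeds the cue, using the jukebox attenuation asset and the quest music sound class at boosted volume.
# A concurrency override allows only one instance at a time, stopping the oldest when a new one starts.
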SoundCue /Game/Athena/Sounds/Events/Jukebox/Jukebox_Music_ChangeTrack_Cue
root=WavePlayer_0
nodes [{id=WavePlayer_0, properties {soundWaveAssetPtr=&/Jukebox_ChangeTrack_01}}]
properties {
 attenuationSettings=&/Jukebox_Athena_Att
 bOverrideConcurrency=true
 maxDistance=1850
 soundClassObject=/Game/Sounds/FortSoundClassesAndModes/Fort_Music_Quest_PSM
 volumeMultiplier=2
 concurrencyOverrides {maxCount=1, resolutionRule=StopOldest}
}
